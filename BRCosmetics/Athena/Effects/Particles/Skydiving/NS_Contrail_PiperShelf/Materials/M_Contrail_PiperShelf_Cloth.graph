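# Lit material for the cloth meshes in the PiperShelf skydiving contrail, used on Niagara mesh particles.
# Blends packed-mask fabric with a diamond shader through a triangle-pattern erosion mask driven by the Crystal Cloth Lerp dynamic parameter.
# A time-offset sine builds flapping WPO and matching normal rotation around a horizontal camera-to-particle axis.
# UV0 serves the diamond function, UV1 baked maps, UV2 patterns; Low/Medium quality flattens diamond surface detail.
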Material /BRCosmetics/Athena/Effects/Particles/Skydiving/NS_Contrail_PiperShelf/Materials/M_Contrail_PiperShelf_Cloth
graph {
 properties {bAutomaticallySetUsageInEditor=false, bUsedWithNiagaraMeshParticles=true, usedShadingModels=Default Lit}
 nodes [
  n1 = VectorParameter() {parameterName=Fabric Center, defaultValue=<rgba 0.72 0.828 0.9 0.888>}
  n2 = AppendVector(A=@n1.RGB, B=@n1.RGB)
  n3 = VectorParameter() {parameterName=Fabric Edges, defaultValue=<rgba 0.7 0.79 1 0.666>}
  n4 = AppendVector(A=@n3.RGB, B=@n3.RGB)
  n5 = TextureSample(UVs=TextureCoordinate() {coordinateIndex=1}) {
   samplerType=SAMPLERTYPE_LinearColor
   texture=/BRCosmetics/Athena/Effects/Particles/Skydiving/NS_Contrail_PiperShelf/Textures/T_Contrail_PiperShelf_Masks
  }
  n6 = Constant() {r=1}
  n7 = OneMinus(None=Multiply(A=OneMinus(None=@n5.A), B=ScalarParameter() {defaultValue=0.666, parameterName=Ambient Occlusion Base Color Contribution}))
  n8 = NamedRerouteDeclaration(None=Multiply(
    A=LinearInterpolate(A=@n2, B=@n4, Alpha=@n5.B)
    B=MaterialFunctionCall(R=@n7, G=@n7, B=@n7, A=@n6) {materialFunction=&AppendMany}.RGBA
   )) {name=Fabric Color and Roughness, nodeColor=<rgba 0 0.7422 1 1>}
  n9 = DeriveNormalZ(InXY=ConstantBiasScale(None=ComponentMask(None=@n5.RGB) {g=true, r=true}) {bias=-0.5, scale=2})
  n10 = ScalarParameter() {defaultValue=0.666, parameterName=Normal Map Flatten}
  n11 = MakeMaterialAttributes!(
   BaseColor=ComponentMask(None=@n8) {b=true, g=true, r=true}
   Metallic=NamedRerouteDeclaration(None=@n5.B) {name=Fabric Metallic, nodeColor=<rgba 0 1 0.6953 1>}
   Specular
   Roughness=ComponentMask(None=@n8) {a=true}
   Anisotropy
   EmissiveColor
   Opacity
   OpacityMask
   Normal=NamedRerouteDeclaration(None=MaterialFunctionCall(Normal=@n9, Flatness=@n10) {materialFunction=&FlattenNormal}.Result) {name=Fabric Normal, nodeColor=<rgba 1 0 0.9844 1>}
   Tangent
   WorldPositionOffset
   SubsurfaceColor
   ClearCoat
   ClearCoatRoughness
   AmbientOcclusion=@n5.A
   Refraction
   CustomizedUVs_0
   CustomizedUVs_1
   CustomizedUVs_2
   CustomizedUVs_3
   CustomizedUVs_4
   CustomizedUVs_5
   CustomizedUVs_6
   CustomizedUVs_7
   PixelDepthOffset
   ShadingModel
   Displacement
  )
  n12 = MakeMaterialAttributes(
   BaseColor=VectorParameter() {parameterName=Diamond Base Color, defaultValue=<rgba 0.49 0.6335 0.7 0>}.RGB
   Metallic=ScalarParameter() {defaultValue=1, parameterName=Diamond Metallic}
   Roughness=NamedRerouteUsage()
   Normal=NamedRerouteUsage()
  )
  n13 = NamedRerouteUsage()
  n14 = BreakMaterialAttributes!(Attr=MaterialFunctionCall!(Material=BlendMaterialAttributes(A=@n11, B=@n12, Alpha=@n13), TechArtMask=@n13) {materialFunction=&/MF_NewDiamond_Contrail}.Material)
  MaterialFunctionCall(Emissive=@n14.EmissiveColor) {materialFunction=&MF_FX_EmissiveAdjustments}
  n15 = NamedRerouteUsage()
  n16 = Multiply(
   A=Add(A=Time(), B=DynamicParameter() {defaultValue=<rgba 0.5 1 1 1>, paramNames [Crystal Cloth Lerp, Time Offset, "", ""]}.Time Offset)
   B=ScalarParameter() {defaultValue=2, parameterName=Flapping Speed}
  )
  n17 = NamedRerouteUsage()
  n18 = Multiply(
   A=ComponentMask(None=TextureCoordinate() {coordinateIndex=2}) {g=true}
   B=Multiply(A=ScalarParameter() {defaultValue=2, parameterName=Flapping Time Offset Along Length}, B=@n17)
  )
  n19 = Multiply(A=Sine(None=Add(A=@n16, B=@n18)), B=Multiply(A=ScalarParameter() {defaultValue=0.007, parameterName=Rotation Range}, B=@n17)) {constB=0}
  NamedRerouteDeclaration(None=RotateAboutAxis(
    NormalizedRotationAxis=@n15
    RotationAngle=@n19
    PivotPoint=MaterialFunctionCall() {materialFunction=&ComponentPivotLocation}.Result
    Position=WorldPosition().XYZ
   )) {name=WPO, nodeColor=<rgba 0 0.1094 1 1>}
  n20 = TextureCoordinate() {coordinateIndex=2}
  n21 = VectorParameter() {parameterName=Triangles Pattern Tiling, defaultValue=<rgba 0.3 3 0.5 5>}
  n22 = ComponentMask(None=@n21.RGB) {g=true, r=true}
  n23 = Multiply(A=@n20, B=@n22)
  n24 = ScalarParameter() {defaultValue=0.62, parameterName=Triangles Layer 1 Weight}
  n25 = Multiply(
   A=TextureSample(UVs=@n23) {samplerType=SAMPLERTYPE_Normal, texture=/Game/Global/Materials/MaterialGym/Materials/TechArtPatterns/Textures/T_Triangles_N}.RGB
   B=@n24
  )
  n26 = AppendVector(A=@n21.B, B=@n21.B)
  n27 = Multiply(A=@n20, B=@n26)
  n28 = ScalarParameter() {defaultValue=1, parameterName=Triangles Layer 2 Weight}
  n29 = Multiply(
   A=TextureSample(UVs=@n27) {samplerType=SAMPLERTYPE_Normal, texture=&T_Triangles_N}.RGB
   B=@n28
  )
  n30 = Add(A=@n24, B=@n28)
  n31 = Normalize(VectorInput=Divide(A=Add(A=@n25, B=@n29), B=@n30))
  n32 = ScalarParameter() {defaultValue=0.9, parameterName=Diamond Normal Flatness}
  n33 = Constant3Vector() {constant=<rgba 0 0 1 0>}
  NamedRerouteDeclaration(None=QualitySwitch(
    Default=MaterialFunctionCall(Normal=@n31, Flatness=@n32) {materialFunction=&FlattenNormal}.Result
    Low=@n33
    Medium=@n33
   )) {name=Diamond Surface Normal TS, nodeColor=<rgba 1 0.8203 0 1>}
  n34 = ScalarParameter() {parameterName=Diamond Surface Roughness Min}
  n35 = ScalarParameter() {defaultValue=0.3, parameterName=Diamond Surface Roughness Max}
  n36 = TextureSample(UVs=@n23) {samplerType=SAMPLERTYPE_Alpha, texture=/Game/Global/Materials/MaterialGym/Materials/TechArtPatterns/Textures/T_Triangles_DF}
  n37 = TextureSample(UVs=@n27) {samplerType=SAMPLERTYPE_Alpha, texture=&T_Triangles_DF}
  n38 = Divide(A=Add(A=Multiply(A=@n36.R, B=@n24), B=Multiply(A=@n37.R, B=@n28)), B=@n30)
  n39 = Constant() {r=0.15}
  NamedRerouteDeclaration(None=QualitySwitch(Default=LinearInterpolate(A=@n34, B=@n35, Alpha=@n38), Low=@n39, Medium=@n39)) {name=Diamond Roughness, nodeColor=<rgba 0 0.0859 1 1>}
  NamedRerouteDeclaration(None=Transform(None=@n14.Normal)) {name=Normal Map WS, nodeColor=<rgba 1 0 0.4453 1>}
  n40 = ComponentMask(None=Subtract(A=CameraPositionWS(), B=ParticlePositionWS())) {g=true, r=true}
  NamedRerouteDeclaration(None=AppendVector(A=Normalize(VectorInput=@n40), B=Constant())) {name=Rotation Axis, nodeColor=<rgba 0.6875 1 0 1>}
  n41 = NamedRerouteUsage()
  n42 = Add(A=RotateAboutAxis(NormalizedRotationAxis=@n15, RotationAngle=@n19, PivotPoint=Constant(), Position=@n41), B=@n41)
  NamedRerouteDeclaration(None=Transform(None=Normalize(VectorInput=@n42)) {transformSourceType=TRANSFORMSOURCE_World, transformType=TRANSFORM_Tangent}) {name=Normal TS, nodeColor=<rgba 0.9219 0 1 1>}
  n43 = TextureSample(UVs=@n23) {samplerType=SAMPLERTYPE_Masks, texture=/Game/Global/Materials/MaterialGym/Materials/TechArtPatterns/Textures/T_Triangles_UV}
  n44 = Add(A=Divide(A=Floor(None=@n23), B=@n22), B=Divide(A=ComponentMask(None=@n43.RGB) {g=true, r=true}, B=@n22))
  n45 = ConstantBiasScale(None=ComponentMask(None=@n44) {r=true}) {bias=-0.5, scale=2}
  n46 = OneMinus(None=Multiply(A=Abs(None=@n45)) {constB=0.5})
  n47 = ComponentMask(None=@n44) {g=true}
  n48 = TextureSample(UVs=@n27) {samplerType=SAMPLERTYPE_Masks, texture=&T_Triangles_UV}
  n49 = Add(A=Divide(A=Floor(None=@n27), B=@n26), B=Divide(A=ComponentMask(None=@n48.RGB) {g=true, r=true}, B=@n26))
  n50 = ConstantBiasScale(None=ComponentMask(None=@n49) {r=true}) {bias=-0.5, scale=2}
  n51 = OneMinus(None=Multiply(A=Abs(None=@n50)) {constB=0.5})
  n52 = ComponentMask(None=@n49) {g=true}
  n53 = Add(A=Multiply(A=Multiply(A=@n46, B=@n47), B=@n24), B=Multiply(A=Multiply(A=@n51, B=@n52), B=@n28))
  n54 = DynamicParameter() {defaultValue=<rgba 0.5 1 1 1>, paramNames [Crystal Cloth Lerp, Time Offset, "", ""]}
  n55 = Power(
   Base=OneMinus(None=Saturate(None=@n54.Crystal Cloth Lerp))
   Exp=ScalarParameter() {defaultValue=1, parameterName=Crystal Cloth Lerp Exponent}
  )
  NamedRerouteDeclaration(None=MaterialFunctionCall!(
    Gradient=Divide(A=@n53, B=@n30)
    Erosion Amount=NamedRerouteDeclaration(None=@n55) {name=Crystal Cloth Lerp, nodeColor=<rgba 0.8047 0 1 1>}
    Softness=ScalarParameter() {defaultValue=0.05, parameterName=Crystal Cloth Erosion Softness}
   ) {materialFunction=/BRCosmetics/Athena/Effects/Particles/Emotes/NS_Emote_Spectrum/Functions/MF_Erosion}.Increased Range) {name=Crystal Cloth Erosion Alpha, nodeColor=<rgba 0 1 0.2266 1>}
  NamedRerouteUsage()
  NamedRerouteUsage()
  TextureCoordinate() {desc="UV0 of the mesh is used only for the diamond material function, since I did not want to touch the TA function to support different UVs"}
  TextureCoordinate() {coordinateIndex=1, desc="UV1 is used for the texture we're sampling here, that has packed normals, a material mask and baked AO"}
  TextureCoordinate() {
   coordinateIndex=2
   desc="UV2 is a 0-1 straight up planar projection before twisting the cloths, I'm using this to generate the blending masks, sample the TA tiling noises etc."
  }
 ]
}
