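# Skydiving contrail cosmetic PiperShelf: two trails of cloth, flickering sparkles, diamond shards and light rays.
# All emitters partition particles across sockets, angled by system alignment direction plus opposing side offsets.
# Cloth and sparkles share an emitter split by spawn group, with sparkles sampled on the cloth mesh and rotated by its quaternion.
# System timeline gates spawning and fades; time of day tints everything, and front-end modifiers adjust the rays.
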NiagaraSystem /BRCosmetics/Athena/Effects/Particles/Skydiving/NS_Contrail_PiperShelf/NS_Contrail_PiperShelf
dependencies {
 usedDataInterfaces [
  &NiagaraDataInterfaceCurve
  &NiagaraDataInterfaceArrayFloat3
  &NiagaraDataInterfaceStaticMesh
  &NiagaraDataInterfaceArrayFloat
  &NiagaraDataInterfaceCamera
  &NiagaraDataInterfaceSocketReader
 ]
 usedDynamicInputs [
  &MakeCustomLinearColorFromBool
  &RandomBool
  &MaskIntBySpawnGroup
  &ReturnExecIndex
  &OrientMeshToVector
  &ReturnCameraVector
  &Multiply_Vector
  &Lerp_Vector
  &TransformVector
  &Add_Float
  &Lerp_Float
  &Sine
  /Niagara/DynamicInputs/Math/Clamp_Vector
  &NormalizeVector
 ]
 usedModules [
  &self:SystemSpawnScript.SystemScriptSource.SystemScriptGraph.NiagaraNodeAssignment_0.SetVariables_#1
  &self:SystemSpawnScript.SystemScriptSource.SystemScriptGraph.NiagaraNodeAssignment_4.SetVariables_#2
  &SystemState
  &self:SystemSpawnScript.SystemScriptSource.SystemScriptGraph.NiagaraNodeAssignment_6.SetVariables_#3
  &self:SystemSpawnScript.SystemScriptSource.SystemScriptGraph.NiagaraNodeAssignment_7.SetVariables_#4
  &Timeline
  &self:SystemSpawnScript.SystemScriptSource.SystemScriptGraph.NiagaraNodeAssignment_3.SetVariables_#5
  &self:SystemSpawnScript.SystemScriptSource.SystemScriptGraph.NiagaraNodeAssignment_8.SetVariables_#6
 ]
}
emitters {
 Cloth_and_Sparkles {
  emitterUpdateScript [
   &EmitterState(
    Inactive Response=::NewEnumerator1
    Life Cycle Mode=::NewEnumerator1
    Loop Behavior=::NewEnumerator1
    Loop Duration Mode=::NewEnumerator1
    UseLoopDelay=false
   )
   &SpawnBurst_Instantaneous(Spawn Count=2, Use Spawn Probability=false, UseLoopCountLimit=false)
   &SpawnRate(Spawn Probability=$System.Timeline.Time, SpawnGroup=1, SpawnRate=50, Use Spawn Probability=true)
  ]
  particleSpawnScript [
   &PartitionParticles()
   &InitializeParticle(
    Color=<rgba 3 3 3 1>
    Lifetime Max=1.3
    Lifetime Min=0.3
    Lifetime Mode=::NewEnumerator1
    Mesh Scale Mode=::NewEnumerator3
    Position Mode=::NewEnumerator0
    Sprite Size Max=<33 66>
    Sprite Size Min=<17 33>
    Sprite Size Mode=::NewEnumerator2
    Write Material Random=true
   )
   /SetVariables(
    Particles.CameraOffset:NiagaraFloat=&MaskFloatBySpawnGroup(Masked Float=0, Passthrough Float=10)
    Particles.VisibilityTag=&SpawnGroup()
   ) {isSetParameters=true}
   &InitialMeshOrientation(Mesh Axis To Orient=<0 0 1>, Mesh Orientation Mode=::NewEnumerator2, Use Rotation Vector=false)
  ]
  particleUpdateScript [
   /SetVariables(Particles.Flicker:NiagaraFloat=&Waveform(
     Global Input Position=&RandomRangeFloat(Minimum=0, Maximum=9)
     "Waveform [1]"=::NewEnumerator10
     "[1] Frequency"=9
     "[1] Amplitude Min/Max"=<0.6 1>
    )) {isSetParameters=true}
   &ParticleState(Kill Particles When Lifetime Has Elapsed=false, Let Infinitely Lived Particles Die When Emitter Deactivates=true)
   &ScaleSpriteSize(
    Uniform Curve Scale=$Particles.Flicker
    Uniform Curve Sprite Scale=DataInterface() {propertyValues {Curve {keys [
        {arriveTangent=9.0725, interpMode=RCIM_Cubic, leaveTangent=9.0725, tangentMode=RCTM_User, time=0, value=0}
        {arriveTangent=-2.5249, interpMode=RCIM_Cubic, leaveTangent=-2.5249, tangentMode=RCTM_User, time=0.1408, value=0.9847}
        {arriveTangent=-0.0405, interpMode=RCIM_Cubic, leaveTangent=-0.0405, tangentMode=RCTM_User, time=1, value=0}
       ]}}}
   )
   &ScaleMeshSize(Scale Factor=&VectorFromFloat(Value=&FloatFromCurve(
      FloatCurve=DataInterface() {propertyValues {Curve {keys [
          {arriveTangent=1.7685, interpMode=RCIM_Cubic, leaveTangent=1.7685, tangentMode=RCTM_User, time=0.1, value=0}
          {arriveTangent=0.0224, interpMode=RCIM_Cubic, leaveTangent=0.0224, tangentMode=RCTM_User, time=1, value=1}
         ]}}}
      CurveIndex=$Emitter.Age
     )))
   &KillParticles(Kill Particles=/Niagara/DynamicInputs/Helpers/MaskBoolBySpawnGroup(
     Masked Bool:NiagaraBool=false
     Passthrough Bool:NiagaraBool=&SetBoolByFloatComparison(A=$Particles.Age, B=$Particles.Lifetime)
     Spawn Group Mask:NiagaraInt32=0
    ))
   &SocketLocation(
    Orientation Sampling=::NewEnumerator1
    Position Sampling=::NewEnumerator1
    SamplePosOffset=false
    Socket Index=$Particles.PartitionParticles.Partition
    Socket Reader=$System.Socket Reader
    Socket Sampling Mode=::NewEnumerator3
    VelocitySpeedLimit=false
   )
   &UpdateMeshOrientation(
    Facing Direction=&Add_Vector(
     A=$System.AlignmentDirection
     B=&SelectVectorFromArray(
      Direct Array Index=$Particles.PartitionParticles.Partition
      Vector Selection Array=DataInterface() {propertyValues {FloatData [<0.2 0 0>, <-0.2 0 0>]}}
     )
    )
    Facing Interpolation=false
    Orientation Method=::NewEnumerator1
    Override Quat=false
    Stabilization Mode:ENiagara_OrientVectorCount=::NewEnumerator2
    TransformFacing=false
    TransformSide=false
   )
   /SetVariables(Particles.ClothQuaternion:Quat4f=&MultiplyQuaternion(
     Quaternion A=$Particles.MeshOrientation
     Quaternion B=&MakeQuatFromAxisAngle(
      Axis=<0 0 1>
      AngleInDegrees=&MergeHelper_FloatSwitch(0=$System.ClothRotationAngleA, 1=$System.ClothRotationAngleB, 2=?, 3=?, Index=$Particles.PartitionParticles.Partition, Num=::NewEnumerator0)
     )
    )) {isSetParameters=true}
   /Niagara/Modules/Spawn/Location/StaticMeshLocation(
    MaskSpawnGroup=false
    Mesh Sampling Type:ENiagara_StaticSamplingMode=::NewEnumerator2
    OffsetAlongNormal=false
    Orientation Sampling:ENiagara_AttributeSamplingApplyOutput=::NewEnumerator1
    OverrideRandomSeed=false
    Position Sampling:ENiagaraSkelMeshPositionSamplingMode=::NewEnumerator2
    SamplePosOffset=false
    Sampled Position Offset:Vector3f=$Output.SocketLocation.SampledPosition
    Static Mesh=DataInterface() {propertyValues {
      DefaultMesh=/BRCosmetics/Athena/Effects/Particles/Skydiving/NS_Contrail_PiperShelf/Meshes/SM_Contrail_PiperShelf_Cloth
      InstanceIndex=-1
      LODIndex=0
      PreviewMesh=&SM_Contrail_PiperShelf_Cloth
      SoftSourceActor=null
      SourceMode=DefaultMeshOnly
      bAllowSamplingFromStreamingLODs=false
      bCaptureTransformsPerFrame=true
      bUsePhysicsBodyVelocity=false
      FilteredSockets []
      LODIndexUserParameter {parameter {name=None, typeDefHandle {registeredTypeIndex=108}, varData []}}
      MeshParameterBinding {parameter {name=None, typeDefHandle {registeredTypeIndex=1}, varData []}}
      SectionFilter {allowedMaterialSlots []}
     }}
   )
   &DynamicMaterialParameters(
    Index 0 Param 1=$System.Timeline.Time
    Index 0 Param 2=&SelectFloatFromArray(Direct Array Index=$Particles.PartitionParticles.Partition, Float Selection Array=DataInterface() {propertyValues {FloatData [0,0.333]}})
    Module.Index0Param0Write=true
    Module.Index0Param1Write=true
    Module.Index0Param2Write=false
    Module.Index0Param3Write=false
   )
   /SetVariables(Particles.Position=&MaskVectorBySpawnGroup(
     Masked Vector=$Output.SocketLocation.SampledPosition
     Passthrough Vector=&AddVectorToPosition(
      Position=$Output.SocketLocation.SampledPosition
      Vector=&RotateVectorByQuaternion(Vector To Rotate=$Particles.StaticMeshLocation.SampledPosition, Quaternion=$Particles.ClothQuaternion)
     )
    )) {isSetParameters=true}
   &ScaleColor(
    Scale Alpha=&FloatFromCurve(
     FloatCurve=DataInterface() {propertyValues {Curve {keys [
         {arriveTangent=9.0725, interpMode=RCIM_Cubic, leaveTangent=9.0725, tangentMode=RCTM_User, time=0, value=0}
         {arriveTangent=-2.5249, interpMode=RCIM_Cubic, leaveTangent=-2.5249, tangentMode=RCTM_User, time=0.1408, value=0.9847}
         {arriveTangent=-0.0405, interpMode=RCIM_Cubic, leaveTangent=-0.0405, tangentMode=RCTM_User, time=1, value=0}
        ]}}}
     Scale Curve=&Multiply_Float(A=$Particles.Flicker, B=$System.Timeline.Time)
    )
    Scale RGB=&MakeVectorFromLinearColorRGB(Color=&FNTimeOfDayColor(Night Color=<rgba 0.17 0.6127 1 1>, Morning Color=<rgba 1 1 1 1>, Day Color=<rgba 1 1 1 1>, Evening Color=<rgba 1 1 1 1>))
    ScaleA=true
    ScaleRGB=true
   )
   /SetVariables(Particles.BackSparkleColor:LinearColor=&MultiplyLinearColor(Linear Color=$Particles.Color, Scale Factor=&MakeVector4_FromFloat(Float=0.1))) {isSetParameters=true}
  ]
  rendererClasses [&NiagaraMeshRendererProperties, &NiagaraSpriteRendererProperties]
  renderers [
   {rendererClass=&NiagaraMeshRendererProperties, rendererIndex=0}
   {rendererClass=&NiagaraSpriteRendererProperties, rendererIndex=1}
   {rendererClass=&NiagaraSpriteRendererProperties, rendererIndex=2}
  ]
 }
 Diamond_Shards {
  emitterUpdateScript [
   &EmitterState(
    Inactive Response=::NewEnumerator1
    Life Cycle Mode=::NewEnumerator1
    Loop Behavior=::NewEnumerator1
    Loop Duration Mode=::NewEnumerator1
    UseLoopDelay=false
   )
   &SpawnRate(Spawn Probability=$System.Timeline.Time, SpawnRate=50, Use Spawn Probability=true)
  ]
  particleSpawnScript [
   &InitializeParticle(
    Color Maximum=<rgba 0.26 0.5 0.4548 1>
    Color Minimum=<rgba 0.4353 0.351 0.5 1>
    Color Mode=::NewEnumerator2
    Lifetime Max=0.4
    Lifetime Min=0.3
    Lifetime Mode=::NewEnumerator1
    Mesh Scale=&RandomRangeVector(Minimum=<0.15 0.15 0.15>, Maximum=<0.3 0.3 0.3>)
    Mesh Scale Mode=::NewEnumerator4
    UsePositionOffset=false
    Write Material Random=true
   )
   &PartitionParticles()
   &InitialMeshOrientation(Use Rotation Vector=false)
   &SocketLocation(
    Orientation Sampling=::NewEnumerator1
    SamplePosOffset=false
    Socket Index=$Particles.PartitionParticles.Partition
    Socket Reader=$System.Socket Reader
    Socket Sampling Mode=::NewEnumerator3
    VelocitySpeedLimit=false
   )
   &ShapeLocation(
    Cylinder Height=150
    Cylinder Height Midpoint=-0.1
    Cylinder Radius=15
    MaskSpawnGroup=false
    Override Seed=false
    Rotation Axis=&Add_Vector(
     A=$System.AlignmentDirection
     B=&SelectVectorFromArray(
      Direct Array Index=$Particles.PartitionParticles.Partition
      Direct Array Mode=::NewEnumerator1
      Vector Selection Array=DataInterface() {propertyValues {FloatData [<0.2 0 0>, <-0.2 0 0>]}}
     )
    )
    Rotation Mode=::NewEnumerator1
    Shape Primitive=::NewEnumerator1
    Spawn Group Mask=1
    Surface Only=false
    UseOwnerScale=true
   )
   &AddVelocity(
    Cone Angle=30
    Cone Axis=&Add_Vector(
     A=$System.AlignmentDirection
     B=&SelectVectorFromArray(
      Direct Array Index=$Particles.PartitionParticles.Partition
      Direct Array Mode=::NewEnumerator1
      Vector Selection Array=DataInterface() {propertyValues {FloatData [<0.2 0 0>, <-0.2 0 0>]}}
     )
    )
    Module.UseConeFalloff=false
    Module.UseVelDistribution=false
    Override Seed=false
    Velocity Mode=::NewEnumerator2
    Velocity Speed=&RandomRangeFloat(Minimum=100, Maximum=200)
   )
  ]
  particleUpdateScript [
   &ParticleState()
   &Drag(Drag=&RandomRangeFloat(Minimum=2.5, Maximum=3), Use Linear Drag=true, Use Rotational Drag=false)
   &CurlNoiseForce(Noise Strength=&RandomRangeFloat(Minimum=250, Maximum=500), Pan Noise=false)
   &SolveForcesAndVelocity(Clamp Velocity=false, Limit Acceleration=false)
   &ScaleMeshSize(Scale Factor=&VectorFromFloat(Value=&FloatFromCurve(
      FloatCurve=DataInterface() {propertyValues {Curve {keys [<at 0.2 1 RCIM_Cubic>, {arriveTangent=-3.8833, interpMode=RCIM_Cubic, tangentMode=RCTM_Break, time=1, value=0}]}}}
      Scale Curve=2
     )))
   &ScaleColor(
    Scale Alpha=&Multiply_Float(
     A=&RandomRangeFloat(Minimum=0.2, Maximum=1)
     B=&Multiply_Float(
      A=&FloatFromCurve(FloatCurve=DataInterface() {propertyValues {Curve {keys [
           {arriveTangent=10.0596, interpMode=RCIM_Cubic, leaveTangent=10.0596, tangentMode=RCTM_User, time=0, value=0}
           {arriveTangent=-0.0111, interpMode=RCIM_Cubic, leaveTangent=-0.0111, tangentMode=RCTM_User, time=0.2, value=1}
           {arriveTangent=-0.0009, interpMode=RCIM_Cubic, leaveTangent=-0.0009, tangentMode=RCTM_User, time=0.9, value=1}
           {arriveTangent=-15.4323, interpMode=RCIM_Cubic, leaveTangent=-15.4323, tangentMode=RCTM_User, time=1, value=0}
          ]}}})
      B=&FNTimeOfDayScalar(Night Scale=0.8, Morning Scale=0.3, Evening Scale=0.5)
     )
    )
    Scale RGB=&Multiply_VectorByFloat(
     Vector=&MakeVectorFromLinearColorRGB(Color=&FNTimeOfDayColor(
       Night Color=<rgba 0.24 0.66 0.8 1>
       Morning Color=<rgba 0.8 0.5145 0.4097 1>
       Day Color=<rgba 1 1 1 1>
       Evening Color=<rgba 1 0.6532 0.5221 1>
      ))
     Float=&RandomRangeFloat(Minimum=1, Maximum=2)
    )
    ScaleA=true
    ScaleRGB=true
   )
  ]
  rendererClasses [&NiagaraMeshRendererProperties]
  renderers [{rendererClass=&NiagaraMeshRendererProperties, rendererIndex=0}]
 }
 LightRay {
  emitterUpdateScript [
   &EmitterState(
    Inactive Response=::NewEnumerator1
    Life Cycle Mode=::NewEnumerator1
    Loop Behavior=::NewEnumerator1
    Loop Duration Mode=::NewEnumerator1
    UseLoopDelay=false
   )
   &SpawnBurst_Instantaneous(Spawn Count=2, Use Spawn Probability=false, UseLoopCountLimit=false)
  ]
  particleSpawnScript [
   &PartitionParticles()
   &InitializeParticle(
    Color Mode=::NewEnumerator0
    Lifetime=&RandomRangeFloat(Minimum=1.7, Maximum=2)
    Mesh Scale=&FN_FrontEndVector3Modifier(Gameplay=<0.75 0.75 1>, Front End=<0.75 0.75 1>)
    Mesh Scale Mode=::NewEnumerator4
    UsePositionOffset=false
    Write Material Random=true
   )
   &InitialMeshOrientation(
    Mesh Axis To Orient=&FN_FrontEndVector3Modifier(Gameplay=<0 0 1>, Front End=<0 0 1>)
    Mesh Orientation Mode=::NewEnumerator2
    Rotation=&FN_FrontEndVector3Modifier(
     Gameplay=&SelectVectorFromArray(
      Direct Array Index=$Particles.PartitionParticles.Partition
      Vector Selection Array=DataInterface() {propertyValues {FloatData [<0.03 0.01 0.09>, <0.03 -0.01 -0.09>, <0 0 0>, <0 0 0>]}}
     )
     Front End=&SelectVectorFromArray(
      Direct Array Index=$Particles.PartitionParticles.Partition
      Vector Selection Array=DataInterface() {propertyValues {FloatData [<0.05 0.03 0.09>, <0.05 -0.03 -0.09>, <0 0 0>, <0 0 0>]}}
     )
    )
    Rotation Coordinate Space=::NewEnumerator2
    Use Rotation Vector=false
   ) {module=InitialMeshOrientation001}
  ]
  particleUpdateScript [
   &ParticleState(Kill Particles When Lifetime Has Elapsed=false, Let Infinitely Lived Particles Die When Emitter Deactivates=true)
   &UpdateMeshOrientation(
    Facing Direction=&Add_Vector(
     A=$System.AlignmentDirection
     B=&SelectVectorFromArray(
      Direct Array Index=$Particles.PartitionParticles.Partition
      Direct Array Mode=::NewEnumerator1
      Vector Selection Array=DataInterface() {propertyValues {FloatData [<0.2 0 0>, <-0.2 0 0>]}}
     )
    )
    Facing Interpolation=false
    Orientation Method=::NewEnumerator1
    Override Quat=false
    Stabilization Mode:ENiagara_OrientVectorCount=::NewEnumerator2
    TransformFacing=false
    TransformSide=false
   ) {module=UpdateMeshOrientation002}
   &ScaleMeshSize(Scale Factor=&MakeVector(
     X=1
     Y=1
     Z=&FloatFromCurve(
      FloatCurve=DataInterface() {propertyValues {Curve {keys [
          {arriveTangent=0.9959, interpMode=RCIM_Cubic, leaveTangent=0.9959, tangentMode=RCTM_User, time=0, value=0}
          {interpMode=RCIM_Cubic, tangentMode=RCTM_User, time=1, value=1}
         ]}}}
      CurveIndex=$Emitter.Age
     )
    ))
   &DynamicMaterialParameters(
    Index 0 Param 2=$System.Dive
    Index 0 Param 3=&RandomRangeFloat(Minimum=0, Maximum=1)
    Index 0 Param 4=&RandomRangeFloat(Minimum=0, Maximum=1)
    Module.Index0Param0Write=true
    Module.Index0Param1Write=true
    Module.Index0Param2Write=true
    Module.Index0Param3Write=true
   ) {module=DynamicMaterialParameters001}
   &SocketLocation(
    Orientation Sampling=::NewEnumerator1
    SamplePosOffset=false
    Socket Index=$Particles.PartitionParticles.Partition
    Socket Reader=$System.Socket Reader
    Socket Sampling Mode=::NewEnumerator3
    VelocitySpeedLimit=false
   )
   &ScaleMeshSize(Scale Factor=&FN_FrontEndVector3Modifier(Gameplay=<0.7 0.7 0.8>, Front End=<0.65 0.65 0.75>)) {module=ScaleMeshSize001}
   &ScaleColor(
    Scale Alpha=&FloatFromCurve(
     FloatCurve=DataInterface() {propertyValues {Curve {keys [
         {interpMode=RCIM_Cubic, tangentMode=RCTM_User, time=0.4, value=0}
         {arriveTangent=0.9919, interpMode=RCIM_Cubic, leaveTangent=0.9919, tangentMode=RCTM_User, time=0.8, value=1}
        ]}}}
     CurveIndex=$System.Timeline.Time
    )
    Scale RGB=&MakeVectorFromLinearColorRGB(Color=&FNTimeOfDayColor(
      Night Color=<rgba 0.1611 0.3212 0.9 1>
      Morning Color=<rgba 1 0.7571 0.6766 1>
      Day Color=<rgba 0.9 0.9 0.9 1>
      Evening Color=<rgba 0.8 0.4131 0.32 1>
     ))
    ScaleA=true
    ScaleRGB=true
   )
   &ScaleColor(Scale Alpha=&FN_FrontEndFloatModifier(Gameplay=1, Front End=0.6), ScaleA=true, ScaleRGB=true) {module=ScaleColor001}
  ]
  rendererClasses [&NiagaraMeshRendererProperties]
  renderers [{rendererClass=&NiagaraMeshRendererProperties, rendererIndex=0}]
 }
}
summary {
 systemName=NS_Contrail_PiperShelf
 emitters [
  {
   emitterName=Cloth_and_Sparkles
   rendererClasses [&NiagaraMeshRendererProperties, &NiagaraSpriteRendererProperties]
  }
  {emitterName=Diamond_Shards, rendererClasses [&NiagaraMeshRendererProperties]}
  {emitterName=LightRay, rendererClasses [&NiagaraMeshRendererProperties]}
 ]
 userVariables [
  User.bIsFrontEnd {type=NiagaraBool}
  User.bIsFrontEndPreview {type=NiagaraBool}
  User.ParaGlideLeanAlpha {type=NiagaraFloat, defaultValue=1, description=""}
 ]
}
